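# Looping daytime exterior ambience cue for the Forge Forest biome, using the outdoor ambient sound class and attenuation.
# A quality-level node sends its first two slots to a quad loop and its last two to a stereo loop.
# Each loop passes through nested, endlessly looping envelopers: a 30-second pitch curve inside a 60-second volume curve.
# The two volume curves differ, with the quad rising slightly and the stereo dipping below full level.
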SoundCue /Game/Sounds/Helios/Biomes/ForgeForest/Ambience_Heroes_Forged_Forest_Day_Ext_Loop_Cue
root=QualityLevel_0
nodes [
 {
  id=Enveloper_0
  children [WavePlayer_1]
  properties {
   bLoop=true
   bLoopIndefinitely=true
   loopEnd=30
   pitchCurve {editorCurveData {keys [{interpMode=RCIM_Cubic, time=0}, <at 15 0.7 RCIM_Cubic>, {interpMode=RCIM_Cubic, time=30}]}}
   volumeCurve {editorCurveData {keys [<at 0 1>]}}
  }
 }
 {
  id=Enveloper_1
  children [WavePlayer_3]
  properties {
   bLoop=true
   bLoopIndefinitely=true
   loopEnd=30
   pitchCurve {editorCurveData {keys [{interpMode=RCIM_Cubic, time=0}, <at 15 0.7 RCIM_Cubic>, {interpMode=RCIM_Cubic, time=30}]}}
   volumeCurve {editorCurveData {keys [<at 0 1>]}}
  }
 }
 {
  id=Enveloper_3
  children [Enveloper_0]
  properties {
   bLoop=true
   bLoopIndefinitely=true
   loopEnd=60
   pitchCurve {editorCurveData {keys [{time=0}]}}
   volumeCurve {editorCurveData {keys [
      {arriveTangent=-0.01, interpMode=RCIM_Cubic, time=0, value=1}
      {arriveTangent=0.0054, interpMode=RCIM_Cubic, leaveTangent=0.0054, time=13.9743, value=1.036}
      {arriveTangent=0.0004, interpMode=RCIM_Cubic, leaveTangent=0.0004, tangentMode=RCTM_User, time=28.2676, value=1.1523}
      {arriveTangent=-0.0004, interpMode=RCIM_Cubic, leaveTangent=-0.0004, tangentMode=RCTM_User, time=35.1185, value=1}
      {arriveTangent=0.006, interpMode=RCIM_Cubic, leaveTangent=0.006, time=41.5626, value=1.1059}
      {arriveTangent=-0.0057, interpMode=RCIM_Cubic, leaveTangent=-0.0057, time=52.6833, value=1.1059}
      {interpMode=RCIM_Cubic, leaveTangent=0.01, time=60, value=1}
     ]}}
  }
 }
 {
  id=Enveloper_4
  children [Enveloper_1]
  properties {
   bLoop=true
   bLoopIndefinitely=true
   loopEnd=60
   pitchCurve {editorCurveData {keys [{time=0}]}}
   volumeCurve {editorCurveData {keys [
      {arriveTangent=-0.01, interpMode=RCIM_Cubic, time=0, value=1}
      {arriveTangent=-0.0022, interpMode=RCIM_Cubic, leaveTangent=-0.0022, time=15, value=0.7}
      {arriveTangent=0, interpMode=RCIM_Cubic, leaveTangent=0, tangentMode=RCTM_User, time=22.2627, value=0.95}
      {arriveTangent=0.0004, interpMode=RCIM_Cubic, leaveTangent=0.0004, tangentMode=RCTM_User, time=29.2932, value=0.8163}
      {arriveTangent=-0.0004, interpMode=RCIM_Cubic, leaveTangent=-0.0004, tangentMode=RCTM_User, time=35.1185, value=1}
      {arriveTangent=-0.023, interpMode=RCIM_Cubic, leaveTangent=-0.023, time=41.3997, value=0.6}
      {arriveTangent=0.0215, interpMode=RCIM_Cubic, leaveTangent=0.0215, time=52.5203, value=0.6}
      {interpMode=RCIM_Cubic, leaveTangent=0.01, time=60, value=1}
     ]}}
  }
 }
 {id=QualityLevel_0, children [Enveloper_3,Enveloper_3,Enveloper_4,Enveloper_4]}
 {
  id=WavePlayer_1
  properties {bLooping=true, soundWaveAssetPtr=&/Ambience_Heroes_Forged_Forest_Day_Ext_Loop_Quad}
 }
 {
  id=WavePlayer_3
  properties {bLooping=true, soundWaveAssetPtr=&/Ambience_Heroes_Forged_Forest_Day_Ext_Loop_Stereo}
 }
]
properties {
 attenuationSettings=&Fort_Ambient_Athena_Att
 maxDistance=2097152
 pitchMultiplier=1.5
 soundClassObject=&Fort_Ambient_Outdoors_Athena
 volumeMultiplier=0.2
}
